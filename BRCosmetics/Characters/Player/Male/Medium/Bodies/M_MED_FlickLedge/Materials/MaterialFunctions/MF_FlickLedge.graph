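# Material function for the FlickLedge character skin that adds animated smoke, wispy hair and an emote color effect to lit color, shaded color and material attributes.
# Two warped, scrolling cylindrical-UV noise samples on pre-skinned positions build the smoke, shaped by gradients and masks, with time-of-day colors.
# Static switches select smoke or hair treatment, driving color, dithered temporal-AA opacity and noise- or sine-based vertex offset.
# An Anim Montage curve scales a spherical emote burst mask.
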MaterialFunction /BRCosmetics/Characters/Player/Male/Medium/Bodies/M_MED_FlickLedge/Materials/MaterialFunctions/MF_FlickLedge
graph {nodes [
  n1 = GetMaterialAttributes(None=FunctionInput() {inputName=Material, inputType=FunctionInput_MaterialAttributes}) {attributeGetTypes [#1, #2]}
  FunctionOutput(None=SetMaterialAttributes!(MaterialAttributes=@n1.MaterialAttributes, OpacityMask=NamedRerouteUsage(), WorldPositionOffset=NamedRerouteUsage(), Opacity) {attributeSetTypes [#1, #2, #3]}) {outputName=Opacity}
  n2 = NamedRerouteUsage()
  n3 = Desaturation(None=@n2, Fraction=Constant() {r=1})
  n4 = NamedRerouteUsage()
  FunctionOutput(None=LinearInterpolate(A=@n2, B=Multiply(A=@n3, B=@n4), Alpha=NamedRerouteUsage())) {outputName=ColorLit}
  n5 = NamedRerouteUsage()
  n6 = Desaturation(None=@n5, Fraction=Constant() {r=1})
  n7 = NamedRerouteUsage()
  FunctionOutput(None=LinearInterpolate(A=@n5, B=Multiply(A=@n6, B=@n7), Alpha=NamedRerouteUsage())) {outputName=ColorShaded}
  n8 = NamedRerouteUsage()
  Multiply(A=@n8) {constB=0.8}
  NamedRerouteDeclaration(None=@n1.OpacityMask) {name=OpacityIn, nodeColor=<rgba 0 0.3437 1 1>}
  NamedRerouteDeclaration(None=@n1.WorldPositionOffset) {name=WPOIn, nodeColor=<rgba 0.1484 0 1 1>}
  n9 = NamedRerouteUsage()
  n10 = VectorParameter() {group=FlickLedge|Smoke, parameterName=SmokePositionOffset, defaultValue=<rgba 0 0 0 1>}
  n11 = MaterialFunctionCall!(
   Circumference Tiling=ScalarParameter() {defaultValue=4, group=FlickLedge|Smoke, parameterName=WarpUV_X_Tiling}
   Position=NamedRerouteDeclaration(None=Add(A=@n9, B=@n10.RGB)) {name=CylinderPosition, nodeColor=<rgba 1 0 0.3281 1>}
   Z Height=ScalarParameter() {defaultValue=40, group=FlickLedge|Smoke, parameterName=WarpUV_Y_Tiling}
   HQ Angle Calculation?
  ) {materialFunction=/Game/Characters/Player/Common/MaterialFunctions/MF_CylindricalUVs_New}
  n12 = TextureSample!(UVs=@n11.UVs, Tex, "DDX(UVs)"=@n11.UVs, "DDY(UVs)"=@n11.UVs, Apply View MipBias) {
   mipValueMode=TMVM_Derivative
   samplerType=SAMPLERTYPE_LinearColor
   texture=/BRCosmetics/Weapons/FORT_Melee/Pickaxe_DyedDuelist/Textures/T_DyedDuelist_Noise
  }
  NamedRerouteDeclaration(None=Subtract(A=@n12.R) {constB=-0.5}) {name=WarpNoise, nodeColor=<rgba 0 1 0.8359 1>}
  NamedRerouteDeclaration(None=Multiply(A=Time(), B=Constant3Vector() {constant=<rgba 0 0 1 0>})) {name=VerticalTime, nodeColor=<rgba 1 0.5156 0 1>}
  n13 = LocalPosition() {localOrigin=InstancePreSkinning}
  NamedRerouteDeclaration(None=ShaderStageSwitch(PixelShader=VertexInterpolator(VS=@n13.XYZ).PS, VertexShader=@n13.XYZ)) {name=PreSkinnedPos, nodeColor=<rgba 0.1953 0 1 1>}
  n14 = ComponentMask(None=NamedRerouteUsage()) {b=true}
  n15 = ScalarParameter() {defaultValue=25.7115, group=FlickLedge|Smoke, parameterName=GradientOffset}
  n16 = Saturate(None=Divide(A=Add(A=@n14, B=@n15) {constB=-100}, B=ScalarParameter() {defaultValue=170, group=FlickLedge|Smoke, parameterName=GradientSmoothness}) {constB=100})
  n17 = ScalarParameter() {defaultValue=1.4597, group=FlickLedge|Smoke, parameterName=GradientExp}
  NamedRerouteDeclaration(None=OneMinus(None=Power(Base=@n16, Exp=@n17))) {name=RAW Vertical Gradient, nodeColor=<rgba 0.4531 1 0 1>}
  n18 = VertexNormalWS()
  n19 = ComponentMask(None=NamedRerouteUsage()) {b=true}
  n20 = Multiply(A=@n19) {constB=-0.5}
  n21 = Multiply(A=Time(), B=ScalarParameter() {defaultValue=4, group=FlickLedge|Hair, parameterName=HairWPOSpeed}) {constB=2}
  n22 = ConstantBiasScale(None=Sine(None=Add(A=@n20, B=@n21)))
  n23 = Multiply(A=@n19) {constB=-0.25}
  n24 = Multiply(A=@n21) {constB=0.5}
  n25 = ConstantBiasScale(None=Sine(None=Add(A=@n23, B=@n24)))
  n26 = Multiply(A=@n25, B=@n25)
  n27 = Add(A=Subtract(A=Multiply(A=@n22, B=@n22)) {constB=0.5}, B=Multiply(A=Subtract(A=@n26) {constB=0.5}) {constB=0.5})
  n28 = Multiply(A=@n19) {constB=-0.05}
  n29 = Multiply(A=@n21) {constB=0.5}
  n30 = Saturate(None=Sine(None=Add(A=@n28, B=@n29)) {period=1.5947})
  n31 = Subtract(A=Multiply(A=@n30, B=@n30)) {constB=0.5}
  n32 = Multiply(
   A=Multiply(A=NamedRerouteUsage(), B=Add(A=@n27, B=@n31))
   B=LinearInterpolate(
    A=ScalarParameter() {defaultValue=0.4, group=FlickLedge|Hair, parameterName=HairWPOIntensityNear}
    B=ScalarParameter() {defaultValue=0.5, group=FlickLedge|Hair, parameterName=HairWPOIntensityFar}
    Alpha=NamedRerouteUsage()
   )
  )
  n33 = Add(
   A=Multiply(A=Multiply(A=@n18, B=@n32), B=ScalarParameter() {defaultValue=0.5, group=FlickLedge|Hair, parameterName=HairWPONormalIntensity})
   B=@n32
  )
  n34 = NamedRerouteUsage()
  NamedRerouteDeclaration(None=StaticSwitchParameter(True=Multiply(A=@n33, B=@n34), False=NamedRerouteUsage()) {group=FlickLedge|Hair, parameterName=IsHair}) {name=HairWPO, nodeColor=<rgba 1 0 0.5156 1>}
  n35 = TextureCoordinate() {coordinateIndex=1}
  NamedRerouteDeclaration(None=Ceil(None=ComponentMask(None=@n35) {r=true})) {name=RimMask, nodeColor=<rgba 1 0.7031 0 1>}
  NamedRerouteDeclaration(None=ChannelMaskParameter(None=FunctionInput() {inputName=InFXMask}) {group=FlickLedge, parameterName=ClothMask}.R) {name=ClothMask, nodeColor=<rgba 0.0312 1 0 1>}
  n36 = Multiply(
   A=Power(Base=NamedRerouteUsage(), Exp=ScalarParameter() {defaultValue=0.75, group=FlickLedge|Smoke, parameterName=VerticalGradientExp})
   B=NamedRerouteUsage()
  )
  n37 = Multiply(
   A=Multiply(A=Time(), B=Constant2Vector() {g=-1})
   B=ScalarParameter() {defaultValue=0.5, group=FlickLedge|Smoke, parameterName=WPONoiseSpeed}
  ) {constB=0.2}
  n38 = MaterialFunctionCall(
   Circumference Tiling=ScalarParameter() {defaultValue=4, group=FlickLedge|Smoke, parameterName=WPOCylindricalUVXTiling}
   Position=NamedRerouteUsage()
   Z Height=ScalarParameter() {defaultValue=-100, group=FlickLedge|Smoke, parameterName=WPOCylindricalUV_Y_Tiling}
  ) {materialFunction=&MF_CylindricalUVs_New}
  n39 = Subtract(A=TextureSample!(UVs=Multiply(A=@n38.UVs) {constB=0.75}, Tex, Apply View MipBias) {samplerType=SAMPLERTYPE_LinearGrayscale, texture=/Game/Global/Textures/Default/Noise/EqualizerComponents_v2_Noise}.R) {constB=0.5}
  n40 = TextureSample(UVs=Add(A=Add(A=@n38.UVs, B=@n37), B=Multiply(A=@n39) {constB=0.1}) {constB=0.1}) {samplerType=SAMPLERTYPE_LinearGrayscale, texture=&EqualizerComponents_v2_Noise}
  n41 = Multiply(
   A=Power(Base=Subtract(A=@n40.R) {constB=-0.5}, Exp=ScalarParameter() {defaultValue=3, group=FlickLedge|Smoke, parameterName=WPONoiseExp})
   B=ScalarParameter() {defaultValue=1, group=FlickLedge|Smoke, parameterName=WPONoiseInt}
  )
  n42 = Multiply(A=VertexNormalWS(), B=@n41)
  NamedRerouteDeclaration(None=StaticSwitchParameter(True=Multiply(A=@n36, B=@n42), False=Constant()) {group=FlickLedge|Smoke, parameterName=EnableSmoke}) {name=SmokeWPO, nodeColor=<rgba 0.3594 0 1 1>}
  n43 = VertexInterpolator(VS=Multiply(A=LocalPosition() {localOrigin=InstancePreSkinning}.XYZ, B=Constant3Vector() {constant=<rgba 1.75 1 1 0>}))
  n44 = VectorParameter() {group=FlickLedge|Emote, parameterName=EmoteMaskPosition, defaultValue=<rgba 0 0 158.5014 1>}
  n45 = ScalarParameter() {defaultValue=45, group=FlickLedge|Emote, parameterName=EmoteMaskRadius}
  n46 = ScalarParameter() {defaultValue=0.666, group=FlickLedge|Emote, parameterName=EmoteMaskHardness}
  NamedRerouteDeclaration(None=Multiply(
    A=SphereMask(A=@n43.PS, B=@n44.RGB, Radius=@n45, Hardness=@n46) {attenuationRadius=35, hardnessPercent=25}
    B=ScalarParameter() {desc=Emote curve driven by the Anim Montage, group=FlickLedge|Emote, parameterName=MaterialCurveA}
   )) {name=EmoteBurstMask, nodeColor=<rgba 1 0.9141 0 1>}
  n47 = ScalarParameter() {group=FlickLedge|Hair, parameterName=HairOpacitySmoothness}
  n48 = NamedRerouteUsage()
  n49 = SphereMask(
   A=Multiply(A=NamedRerouteUsage(), B=Constant3Vector() {constant=<rgba 1.9494 1 1 0>})
   B=VectorParameter() {group=FlickLedge|Hair, parameterName=HairTipsMaskPosition, defaultValue=<rgba 0 0 159 1>}.RGB
   Radius=ScalarParameter() {defaultValue=13, group=FlickLedge|Hair, parameterName=HairTipMaskRadius}
   Hardness=ScalarParameter() {defaultValue=0.85, group=FlickLedge|Hair, parameterName=HairTipMaskHardness}
  ) {attenuationRadius=15, hardnessPercent=75}
  NamedRerouteDeclaration(None=Max(A=InverseLinearInterpolate(A=@n47, B=@n48, Value=@n49), B=@n49)) {name=HairOpacityMask, nodeColor=<rgba 1 0 0.4922 1>}
  n50 = NamedRerouteUsage()
  n51 = NamedRerouteUsage()
  n52 = NamedRerouteUsage()
  n53 = Power(Base=NamedRerouteUsage(), Exp=ScalarParameter() {defaultValue=12, group=FlickLedge|Smoke, parameterName=SmokeNoiseExp}) {constExponent=8}
  n54 = Multiply(
   A=Power(Base=NamedRerouteUsage(), Exp=ScalarParameter() {defaultValue=1, group=FlickLedge|Smoke, parameterName=SmokeGradientOpacityExp}) {constExponent=4}
   B=Add(A=Multiply(A=@n52, B=@n53), B=@n53)
  )
  n55 = SmoothStep(
   Min=ScalarParameter() {group=FlickLedge|Smoke, parameterName=GradientMin01}
   Max=ScalarParameter() {defaultValue=2.5, group=FlickLedge|Smoke, parameterName=GradientMax01}
   Value=NamedRerouteUsage()
  )
  n56 = Multiply(A=Saturate(None=Add(A=@n54, B=@n55)), B=ScalarParameter() {defaultValue=2, group=FlickLedge|Smoke, parameterName=OverallOpacityScale})
  n57 = SmoothStep(
   Min=ScalarParameter() {group=FlickLedge|Smoke, parameterName=OverallOpacityMin}
   Max=ScalarParameter() {defaultValue=0.256, group=FlickLedge|Smoke, parameterName=OverallOpacityMax}
   Value=LinearInterpolate(A=Saturate(None=@n56), B=@n55, Alpha=ScalarParameter() {group=FlickLedge|Smoke, parameterName=IsOutline})
  )
  n58 = LinearInterpolate(A=@n51, B=Min(A=@n51, B=Saturate(None=@n57)), Alpha=NamedRerouteUsage()) {constA=1}
  NamedRerouteDeclaration(None=MaterialFunctionCall(Alpha Threshold=StaticSwitchParameter(True=@n50, False=@n58) {group=FlickLedge, parameterName=IsHair}) {materialFunction=&DitherTemporalAA_Fortnite}.Result) {name=OpacityOut, nodeColor=<rgba 0 0.1094 1 1>}
  NamedRerouteDeclaration(None=MaterialFunctionCall(
    Color_Morning=VectorParameter() {group=FlickLedge|Hair, parameterName=HairTipColor_Morning, defaultValue=<rgba 0.3023 0.8654 1 1>}.RGB
    Color_Day=VectorParameter() {group=FlickLedge|Hair, parameterName=HairTipColor_Day, defaultValue=<rgba 0.3023 0.8654 1 1>}.RGB
    Color_Evening=VectorParameter() {group=FlickLedge|Hair, parameterName=HairTipColor_Evening, defaultValue=<rgba 1 1 1 1>}.RGB
    Color_Night=VectorParameter() {group=FlickLedge|Hair, parameterName=HairTipColor_Night, defaultValue=<rgba 1 1 1 1>}.RGB
   ) {materialFunction=&Color_Time_of_Day}.Result) {name=HairTipColor, nodeColor=<rgba 0 0.0156 1 1>}
  n59 = TextureCoordinate()
  n60 = Constant2Vector() {g=0.05, r=0.94}
  n61 = Constant2Vector() {g=0.85, r=0.125}
  n62 = Constant()
  n63 = LinearInterpolate(
   A=VectorParameter() {group=FlickLedge|Emote, parameterName=EmoteColor, defaultValue=<rgba 1.5 0.1726 0 1>}.RGB
   B=VectorParameter() {group=FlickLedge|Emote, parameterName=EmoteColor2, defaultValue=<rgba 1.2 0.5351 0.1389 1>}.RGB
   Alpha=Multiply(
    A=MaterialFunctionCall!(A=@n59, B=@n60, Bounds=@n61, Edge Falloff=@n62) {materialFunction=/Engine/Functions/Engine_MaterialFunctions01/Gradient/BoxMask-2D}.Result
    B=ScalarParameter() {group=FlickLedge|Emote, parameterName=UseSecondColor}
   )
  )
  n64 = ScalarParameter() {defaultValue=2.7, group=FlickLedge|Emote, parameterName=EmoteColorNight}
  NamedRerouteDeclaration(None=MaterialFunctionCall(Color_Morning=@n63, Color_Day=@n63, Color_Evening=@n63, Color_Night=Multiply(A=@n63, B=@n64)) {materialFunction=&Color_Time_of_Day}.Result) {name=EmoteColor, nodeColor=<rgba 1 0 0.375 1>}
  NamedRerouteDeclaration(None=FunctionInput() {inputName=InColorLit}) {name=ColorLitIn, nodeColor=<rgba 0 1 0.1562 1>}
  NamedRerouteDeclaration(None=FunctionInput() {inputName=InColorShaded}) {name=ColorShadedIn, nodeColor=<rgba 1 0 0.1641 1>}
  n65 = ScalarParameter() {defaultValue=0.54, group=FlickLedge|Smoke, parameterName=NoiseGradientMin}
  n66 = ScalarParameter() {defaultValue=1.2179, group=FlickLedge|Smoke, parameterName=NoiseGradientMax}
  n67 = NamedRerouteUsage()
  n68 = Saturate(None=Multiply(
    A=SmoothStep(Min=@n65, Max=@n66, Value=@n67)
    B=ScalarParameter() {defaultValue=1.5, group=FlickLedge|Smoke, parameterName=NoiseGradientIntensity}
   ))
  n69 = NamedRerouteUsage()
  n70 = Multiply(A=Multiply(A=@n68, B=Add(A=@n68, B=@n69)), B=ScalarParameter() {defaultValue=2, group=FlickLedge|Smoke, parameterName=NoiseMaskIntensity}) {constB=3}
  NamedRerouteDeclaration(None=Multiply(A=Saturate(None=@n70), B=NamedRerouteUsage())) {name=Mask, nodeColor=<rgba 0 0.4141 1 1>}
  n71 = ScalarParameter() {defaultValue=3, group=FlickLedge|Smoke, parameterName=CylindricalUVXTiling}
  n72 = NamedRerouteUsage()
  n73 = NamedRerouteUsage()
  n74 = ScalarParameter() {defaultValue=16, group=FlickLedge|Smoke, parameterName=SmokeSpeed}
  n75 = MaterialFunctionCall(
   Circumference Tiling=@n71
   Position=Add(A=@n72, B=Multiply(A=@n73, B=@n74) {constB=30})
   Z Height=ScalarParameter() {defaultValue=50, group=FlickLedge|Smoke, parameterName=CylindricalUV_Y_Tiling_1}
  ) {materialFunction=&MF_CylindricalUVs_New}
  n76 = NamedRerouteUsage()
  n77 = ScalarParameter() {defaultValue=0.3, group=FlickLedge|Smoke, parameterName=WarpNoiseIntensity}
  n78 = TextureSample!(
   UVs=Add(A=@n75.UVs, B=Multiply(A=@n76, B=@n77) {constB=0.3})
   Tex
   "DDX(UVs)"=@n75.Explicit DDX
   "DDY(UVs)"=@n75.Explicit DDX
   Apply View MipBias
  ) {mipValueMode=TMVM_Derivative, texture=/BRCosmetics/Characters/Player/Male/Medium/Bodies/M_MED_FlickLedge/Textures/T_FlickLedge_Smoke}
  n79 = NamedRerouteUsage()
  n80 = Divide(A=@n74, B=Constant() {r=3}) {constB=3}
  n81 = MaterialFunctionCall(
   Circumference Tiling=Round(None=Divide(A=@n71) {constB=3})
   Position=Add(A=@n72, B=Multiply(A=@n79, B=@n80) {constB=10})
   Z Height=ScalarParameter() {defaultValue=80, group=FlickLedge|Smoke, parameterName=CylindricalUV_Y_Tiling_2}
  ) {materialFunction=&MF_CylindricalUVs_New}
  n82 = NamedRerouteUsage()
  n83 = TextureSample(UVs=Add(A=@n81.UVs, B=Multiply(A=@n82) {constB=-0.2}), "DDX(UVs)"=@n81.Explicit DDX, "DDY(UVs)"=@n81.Explicit DDX) {mipValueMode=TMVM_Derivative, texture=&T_FlickLedge_Smoke}
  n84 = Multiply(A=@n78.R, B=@n83.R)
  NamedRerouteDeclaration(None=Power(Base=OneMinus(None=@n84), Exp=ScalarParameter() {defaultValue=0.5, group=FlickLedge|Smoke, parameterName=SmokeNoiseExponent}) {constExponent=1}) {name=SmokeNoise, nodeColor=<rgba 0.7344 1 0 1>}
  n85 = NamedRerouteUsage()
  n86 = ScalarParameter() {defaultValue=1.2, group=FlickLedge|Smoke, parameterName=SmokeColorMin}
  n87 = ScalarParameter() {defaultValue=2, group=FlickLedge|Smoke, parameterName=SmokeColorMax}
  n88 = NamedRerouteUsage()
  n89 = Multiply(A=NamedRerouteUsage(), B=LinearInterpolate(A=@n86, B=@n87, Alpha=@n88) {constA=0.3, constB=1.3})
  n90 = NamedRerouteUsage()
  NamedRerouteDeclaration(None=StaticSwitchParameter(True=LinearInterpolate(A=@n85, B=@n89, Alpha=@n90), False=NamedRerouteUsage()) {defaultValue=true, group=FlickLedge|Smoke, parameterName=EnableSmoke}) {name=SmokeColorLit, nodeColor=<rgba 0 1 0.6719 1>}
  n91 = NamedRerouteUsage()
  n92 = Multiply(A=@n8, B=LinearInterpolate(A=@n86, B=@n87, Alpha=@n88) {constA=0.3, constB=1.3})
  n93 = NamedRerouteUsage()
  NamedRerouteDeclaration(None=StaticSwitchParameter(True=LinearInterpolate(A=@n91, B=@n92, Alpha=@n93), False=NamedRerouteUsage()) {defaultValue=true, group=FlickLedge|Smoke, parameterName=EnableSmoke}) {name=SmokeColorShaded, nodeColor=<rgba 1 0.3516 0 1>}
  n94 = ScalarParameter() {defaultValue=0.02, group=FlickLedge|Hair, parameterName=HairFresnelExponent}
  n95 = NamedRerouteUsage()
  n96 = Constant() {r=0.35}
  n97 = Subtract(
   A=ScalarParameter() {defaultValue=2, group=FlickLedge|Hair, parameterName=HairMaskExponent}
   B=Multiply(A=MaterialFunctionCall!(Time Mult=@n96) {materialFunction=/Game/Effects/Fort_Effects/MaterialFunctions/MF_Flicker}.Raw Flicker) {constB=0.3}
  )
  n98 = Saturate(None=Multiply(A=Fresnel(ExponentIn=@n94) {baseReflectFraction=0}, B=Power(Base=@n95, Exp=@n97)))
  n99 = InverseLinearInterpolate(
   A=ScalarParameter() {defaultValue=-0.63, group=FlickLedge|Hair, parameterName=HairTextureSmoothness}
   B=Power(Base=NamedRerouteUsage(), Exp=ScalarParameter() {defaultValue=1, group=FlickLedge|Hair, parameterName=HairSmokeExponent})
   Value=@n98
  )
  NamedRerouteDeclaration(None=Saturate(None=Multiply(A=@n99, B=@n98))) {name=HairTextureNoise, nodeColor=<rgba 0.6641 0 1 1>}
  n100 = NamedRerouteUsage()
  n101 = NamedRerouteUsage()
  n102 = LinearInterpolate(
   A=NamedRerouteUsage()
   B=Power(Base=NamedRerouteUsage()) {constExponent=0.1}
   Alpha=ScalarParameter() {group=FlickLedge|Hair, parameterName=IsHead}
  )
  NamedRerouteDeclaration(None=StaticSwitchParameter(True=LinearInterpolate(A=@n100, B=@n101, Alpha=@n102), False=NamedRerouteUsage()) {group=FlickLedge|Hair, parameterName=IsHair}) {name=HairLitColor, nodeColor=<rgba 0.4062 1 0 1>}
  NamedRerouteDeclaration(None=@n78.R) {name=SmokeNoiseRaw, nodeColor=<rgba 1 0 0.8437 1>}
  n103 = NamedRerouteUsage()
  n104 = NamedRerouteUsage()
  NamedRerouteDeclaration(None=StaticSwitchParameter(True=LinearInterpolate(A=@n103, B=@n104, Alpha=@n102), False=NamedRerouteUsage()) {group=FlickLedge|Hair, parameterName=IsHair}) {name=HairShadedColor, nodeColor=<rgba 1 0 0.7266 1>}
  NamedRerouteDeclaration(None=MaterialFunctionCall(
    Color_Morning=VectorParameter() {group=FlickLedge|Smoke, parameterName=SmokeColor_Morning, defaultValue=<rgba 0.0128 0.0144 0.018 1>}.RGB
    Color_Day=VectorParameter() {group=FlickLedge|Smoke, parameterName=SmokeColor_Day, defaultValue=<rgba 0.0128 0.0144 0.018 1>}.RGB
    Color_Evening=VectorParameter() {group=FlickLedge|Smoke, parameterName=SmokeColor_Evening, defaultValue=<rgba 0.0128 0.0144 0.018 1>}.RGB
    Color_Night=VectorParameter() {group=FlickLedge|Smoke, parameterName=SmokeColor_Night, defaultValue=<rgba 0.0128 0.0144 0.018 1>}.RGB
   ) {materialFunction=&Color_Time_of_Day}.Result) {name=SmokeColor, nodeColor=<rgba 0 1 0.2734 1>}
  n105 = ObjectPositionWS()
  n106 = CameraPositionWS()
  n107 = Divide(A=Add(A=Distance(A=@n105, B=@n106), B=ScalarParameter() {defaultValue=-274.7252, group=FlickLedge, parameterName=CameraFade_Offset})) {constB=1024}
  NamedRerouteDeclaration(None=Saturate(None=Divide(A=@n107) {constB=0.01})) {name=CameraFade, nodeColor=<rgba 0 1 0.9766 1>}
  n108 = Subtract(A=ComponentMask(None=@n35) {g=true}) {constB=-0.01}
  NamedRerouteDeclaration(None=Abs(None=ConstantBiasScale(None=@n108) {bias=-0.5, scale=2})) {name=UV2VerticalGradient, nodeColor=<rgba 0.9688 1 0 1>}
  NamedRerouteDeclaration(None=@n41) {name=WPOSmokeTexture, nodeColor=<rgba 1 0.75 0 1>}
  NamedRerouteDeclaration(None=@n68) {name=FXVerticalGradient, nodeColor=<rgba 0.1953 1 0 1>}
  n109 = NamedRerouteUsage()
  n110 = Multiply(
   A=Multiply(A=VertexNormalWS(), B=ScalarParameter() {defaultValue=-5, group=FlickLedge|Hair, parameterName=HeadColapseAmount})
   B=NamedRerouteUsage()
  )
  n111 = ScalarParameter() {group=FlickLedge|Hair, parameterName=IsHead}
  NamedRerouteDeclaration(None=Add(A=LinearInterpolate(A=@n109, B=@n110, Alpha=@n111), B=NamedRerouteUsage())) {name=WPOOut, nodeColor=<rgba 0.1719 1 0 1>}
  n112 = SphereMask(
   A=NamedRerouteUsage()
   B=VectorParameter() {group=FlickLedge|Hair, parameterName=HairMaskPosition, defaultValue=<rgba 0 8.0791 153.9942 1>}.RGB
   Radius=ScalarParameter() {defaultValue=16.2553, group=FlickLedge|Hair, parameterName=HairMaskRadius}
   Hardness=ScalarParameter() {defaultValue=0.75, group=FlickLedge|Hair, parameterName=HairMaskHardness}
  ) {attenuationRadius=15, hardnessPercent=75}
  NamedRerouteDeclaration(None=LinearInterpolate(A=@n112, B=OneMinus(None=@n112), Alpha=ScalarParameter() {defaultValue=1, group=FlickLedge|Hair, parameterName=HairMaskInvert})) {name=HairMask, nodeColor=<rgba 1 0 0.4922 1>}
  NamedRerouteUsage()
 ]}
